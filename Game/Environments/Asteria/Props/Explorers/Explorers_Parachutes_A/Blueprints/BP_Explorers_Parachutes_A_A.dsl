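; Explorer parachute prop Blueprint with a configurable mesh and color, set up at construction time.
; The construction script calls the parent, assigns the chosen mesh to the Parachute component, then runs SelectColor.
; SelectColor maps a color enum to four preset colors or a custom color, which SetColoronSecondaryMesh then applies to the parachute.
; The BeginPlay, overlap and Tick events are disabled stubs, so the Blueprint adds no runtime logic.
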(event EventBeginPlay
  (|CallParentFunction) ; This node is disabled and will not be called.\nDrag off pins to build functionality.
)

(event Collision|EventActorBeginOverlap
  (params (OtherActor :type Actor%20Object%20Reference))
  (|CallParentFunction OtherActor) ; This node is disabled and will not be called.\nDrag off pins to build functionality.
)

(event EventTick
  (params (DeltaSeconds :type Float%20%28single-precision%29))
  (|CallParentFunction DeltaSeconds) ; This node is disabled and will not be called.\nDrag off pins to build functionality.
)

(fn SelectColor
  (switch Utilities|FlowControl|Switch|SwitchonENUM_ParachuteColors (Variables|Default|GetParachuteColor)
    (:NewEnumerator0
      (Variables|Default|SetColor "(R=0.018500,G=0.093059,B=0.027321,A=1.000000)"))
    (:NewEnumerator1
      (Variables|Default|SetColor "(R=0.361307,G=0.283149,B=0.147027,A=1.000000)"))
    (:NewEnumerator2
      (Variables|Default|SetColor "(R=0.104616,G=0.001518,B=0.010960,A=1.000000)"))
    (:NewEnumerator3
      (Variables|Default|SetColor "(R=0.082283,G=0.082283,B=0.082283,A=1.000000)"))
    (:NewEnumerator4
      (Variables|Default|SetColor (Variables|Default|GetCustomColor)))))

(fn ConstructionScript
  (|CallParentFunction)
  (Components|StaticMesh|SetStaticMesh (Variables|Default|GetParachute) (Variables|Default|GetParachuteMesh))
  (CallFunction|SelectColor)
  (Default|SetColoronSecondaryMesh (Variables|Default|GetColor) (Variables|Default|GetParachute)))
